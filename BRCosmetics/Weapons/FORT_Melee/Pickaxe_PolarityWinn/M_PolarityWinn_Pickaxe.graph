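# Unlit translucent effect material for the PolarityWinn pickaxe, usable on Niagara mesh particles, drawing animated energy lines.
# A UV1 sphere mask builds a gradient, while panned noise supplies UV distortion.
# Time-driven, sharpened sine waves form lines, and named reroutes link these masks to outer fade, opacity and a four-color emissive lerp.
# Particle color multiplies color and alpha, and dynamic parameters control distortion and line count.
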Material /BRCosmetics/Weapons/FORT_Melee/Pickaxe_PolarityWinn/M_PolarityWinn_Pickaxe
graph {
 properties {
  bOutputTranslucentVelocity=true
  bUseTranslucencyVertexFog=false
  bUsedWithNiagaraMeshParticles=true
  blendMode=BLEND_Translucent
  pixelDepthOffsetMode=PDOM_Legacy
  shadingModel=MSM_Unlit
  usedShadingModels=Unlit
 }
 nodes [
  n1 = ParticleColor()
  MaterialFunctionCall(Emissive=Multiply(A=NamedRerouteUsage(), B=@n1.RGB)) {materialFunction=&MF_FX_EmissiveAdjustments}
  Multiply(A=NamedRerouteUsage(), B=@n1.A)
  NamedRerouteDeclaration(None=MaterialFunctionCall!(
    A=Constant3Vector()
    B=VectorParameter() {group=Emissive, parameterName=Color_Outer, sortPriority=2, defaultValue=<rgba 0.9155 0.0931 1 1>}.RGB
    C=VectorParameter() {group=Emissive, parameterName=Color_Center, sortPriority=1, defaultValue=<rgba 0.0518 0.1821 1 1>}.RGB
    D=VectorParameter() {group=Emissive, parameterName=Color_Inner, sortPriority=0, defaultValue=<rgba 0.1598 0.8964 1 1>}.RGB
    Alpha=NamedRerouteUsage()
   ) {materialFunction=/BRCosmetics/Characters/Player/Female/Medium/Bodies/F_MED_PolarityWinn_Field/Materials/MaterialFunctions/Lerp_4Color}.Result) {name=Emissive, nodeColor=<rgba 0.7344 1 0 1>}
  n2 = Saturate(None=OneMinus(None=NamedRerouteUsage()))
  NamedRerouteDeclaration(None=Multiply(A=Power(Base=@n2) {constExponent=12}, B=NamedRerouteUsage())) {name=Opacity, nodeColor=<rgba 0.9453 0 1 1>}
  n3 = NamedRerouteUsage()
  n4 = DynamicParameter() {defaultValue=<rgba 1 0.3 6 0.5>, paramNames [Distortion Intensity 1, Distortion Intensity 2, LineAmount, ""]}
  n5 = SphereMask(
   A=TextureCoordinate() {coordinateIndex=1}
   B=Constant() {r=0.5}
   Radius=Add(A=Multiply(A=@n3, B=@n4.Distortion Intensity 2)) {constB=0.5}
   Hardness=Constant()
  )
  n6 = OneMinus(None=ComponentMask(None=TextureCoordinate()) {g=true})
  NamedRerouteDeclaration(None=Multiply(A=Saturate(None=@n5), B=Power(Base=@n6) {constExponent=0.5})) {name=Gradient, nodeColor=<rgba 0 0.1094 1 1>}
  n7 = TextureCoordinate()
  n8 = ComponentMask(None=TextureCoordinate() {coordinateIndex=1}) {g=true}
  n9 = Multiply(
   A=Add(A=ComponentMask(None=@n7) {g=true}) {constB=0.15}
   B=Power(Base=OneMinus(None=@n8), Exp=ScalarParameter() {defaultValue=4, group=UV Distortion, parameterName=DistortionMask, sortPriority=2}) {constExponent=1.5}
  )
  n10 = TextureCoordinate()
  n11 = VectorParameter() {
   group=UV Distortion
   parameterName=Distortion UV
   sortPriority=1
   channelNames {a=Speed V, b=Speed U, g=Tile V, r=Tile U}
   defaultValue=<rgba 9 1 0 1>
  }
  n12 = AppendVector(A=@n11.Tile U, B=@n11.Tile U)
  n13 = TextureSampleParameter2D(UVs=Panner(Coordinate=Multiply(A=@n10, B=@n12), Speed=AppendVector(A=@n11.Speed U, B=@n11.Speed U)) {speedY=-0.25}) {
   group=UV Distortion
   parameterName=Distortion Tex
   samplerType=SAMPLERTYPE_Grayscale
   sortPriority=0
   texture=/BRCosmetics/Effects/Fort_Effects/Emotes/SwissKale/T_Blur_Noise
  }
  n14 = DynamicParameter() {defaultValue=<rgba 1 0.3 6 0.5>, paramNames [Distortion Intensity 1, Distortion Intensity 2, LineAmount, ""]}
  n15 = Multiply(A=@n13.G, B=@n14.Distortion Intensity 1) {constB=0.1}
  NamedRerouteDeclaration(None=Saturate(None=Multiply(A=@n9, B=@n15))) {name=UV Distortion, nodeColor=<rgba 0.6406 0 1 1>}
  n16 = Time()
  n17 = ScalarParameter() {defaultValue=0.2, group=Lines, parameterName=TimeSpeed, sortPriority=1}
  n18 = Add(A=NamedRerouteUsage(), B=NamedRerouteDeclaration(None=Multiply(A=@n16, B=@n17) {constB=0.2}) {name=Time, nodeColor=<rgba 1 0.0469 0 1>})
  n19 = ConstantBiasScale(None=Sine(None=Multiply(A=@n18, B=@n4.LineAmount) {constB=6}))
  n20 = ScalarParameter() {defaultValue=12, group=Lines, parameterName=LineExponent, sortPriority=0}
  NamedRerouteDeclaration(None=Saturate(None=Power(Base=@n19, Exp=@n20) {constExponent=6})) {name=SineWaves, nodeColor=<rgba 0.8281 0 1 1>}
  n21 = VectorParameter() {
   group=Lines
   parameterName=OuterFade Adjustments
   sortPriority=2
   channelNames {a="", b=Mult B, g=Subtract, r=Mult A}
   defaultValue=<rgba 50 75 500 1>
  }
  n22 = ConstantBiasScale(None=NamedRerouteUsage())
  n23 = Power(Base=OneMinus(None=NamedRerouteUsage()), Exp=Subtract(A=@n21.Subtract, B=Multiply(A=@n22, B=@n21.Mult A) {constB=50})) {constExponent=64}
  NamedRerouteDeclaration(None=Multiply(A=Saturate(None=@n23), B=@n21.Mult B) {constB=4}) {name=OuterFade, nodeColor=<rgba 1 0.1875 0 1>}
 ]
}
