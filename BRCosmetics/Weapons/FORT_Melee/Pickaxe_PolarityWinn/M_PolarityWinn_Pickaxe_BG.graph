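# Translucent unlit material for the PolarityWinn pickaxe, usable on Niagara mesh particles, that renders a distorted, color-graded glow shape.
# Panning noise, scaled by dynamic parameters and masked along UV height, builds a UV distortion term.
# A sphere mask on UV1 with a perturbed radius, times a vertical falloff, forms a procedural gradient.
# Remapped masking drives a four-color lerp, with particle color modulating emissive and alpha.
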Material /BRCosmetics/Weapons/FORT_Melee/Pickaxe_PolarityWinn/M_PolarityWinn_Pickaxe_BG
graph {
 properties {
  bOutputTranslucentVelocity=true
  bUseTranslucencyVertexFog=false
  bUsedWithNiagaraMeshParticles=true
  blendMode=BLEND_Translucent
  pixelDepthOffsetMode=PDOM_Legacy
  shadingModel=MSM_Unlit
  usedShadingModels=Unlit
 }
 nodes [
  n1 = ParticleColor()
  MaterialFunctionCall(Emissive=Multiply(A=NamedRerouteUsage(), B=@n1.RGB)) {materialFunction=&MF_FX_EmissiveAdjustments}
  Multiply(A=NamedRerouteUsage(), B=@n1.A)
  n2 = Power(Base=NamedRerouteUsage()) {constExponent=0.25}
  n3 = VectorParameter() {group=Emissive, parameterName=Remap_Values, sortPriority=0, defaultValue=<rgba 0.008 1.113 0.2502 1.0382>}
  NamedRerouteDeclaration(None=MaterialFunctionCall(Input=Multiply(A=@n2) {constB=0.9}, Input Low=@n3.R, Input High=@n3.R, Target Low=@n3.R, Target High=@n3.R) {materialFunction=&RemapValueRange}.Result) {name=EmissiveMask, nodeColor=<rgba 0.4062 0 1 1>}
  NamedRerouteDeclaration(None=MaterialFunctionCall!(
    A=Constant3Vector()
    B=VectorParameter() {group=Emissive, parameterName=Color Outer, sortPriority=3, defaultValue=<rgba 0.9155 0.0931 1 1>}.RGB
    C=VectorParameter() {group=Emissive, parameterName=Color Center, sortPriority=2, defaultValue=<rgba 0.0518 0.1821 1 1>}.RGB
    D=VectorParameter() {group=Emissive, parameterName=Color Inner, sortPriority=1, defaultValue=<rgba 0.1598 0.8964 1 1>}.RGB
    Alpha=NamedRerouteUsage()
   ) {materialFunction=/BRCosmetics/Characters/Player/Female/Medium/Bodies/F_MED_PolarityWinn_Field/Materials/MaterialFunctions/Lerp_4Color}.Result) {name=Emissive, nodeColor=<rgba 0 0.5781 1 1>}
  n4 = NamedRerouteUsage()
  n5 = ScalarParameter() {defaultValue=0.3, group=UV Distortion, parameterName=SDF Distortion, sortPriority=3}
  n6 = SphereMask(A=TextureCoordinate() {coordinateIndex=1}, B=Constant() {r=0.5}, Radius=Add(A=Multiply(A=@n4, B=@n5)) {constB=0.35}, Hardness=Constant())
  n7 = OneMinus(None=ComponentMask(None=TextureCoordinate()) {g=true})
  NamedRerouteDeclaration(None=Multiply(A=Saturate(None=@n6), B=Power(Base=@n7) {constExponent=1})) {name=Gradient, nodeColor=<rgba 0 0.1094 1 1>}
  n8 = TextureCoordinate()
  n9 = ComponentMask(None=TextureCoordinate() {coordinateIndex=1}) {g=true}
  n10 = Multiply(
   A=Add(A=ComponentMask(None=@n8) {g=true}) {constB=0.15}
   B=Power(Base=OneMinus(None=@n9), Exp=ScalarParameter() {defaultValue=4, group=UV Distortion, parameterName=DistortionMask, sortPriority=2}) {constExponent=1.5}
  )
  n11 = TextureCoordinate()
  n12 = DynamicParameter() {defaultValue=<rgba -0.5 3 0.5 1.2>, paramNames [Speed V, Tile U, Tile V, Distortion Intensity]}
  n13 = AppendVector(A=@n12.Tile U, B=@n12.Tile U)
  n14 = VectorParameter() {
   group=UV Distortion
   parameterName=Distortion UV
   sortPriority=1
   channelNames {a=Speed V, b=Speed U, g=Tile V, r=Tile U}
   defaultValue=<rgba 10 0.5 0 -0.5>
  }
  n15 = TextureSampleParameter2D(UVs=Panner(Coordinate=Multiply(A=@n11, B=@n13), Speed=AppendVector(A=@n14.Speed U, B=@n14.Speed U)) {speedY=-0.25}) {
   group=UV Distortion
   parameterName=Distortion Tex
   samplerType=SAMPLERTYPE_Grayscale
   sortPriority=0
   texture=/BRCosmetics/Effects/Fort_Effects/Emotes/SwissKale/T_Blur_Noise
  }
  n16 = Multiply(A=@n15.G, B=@n12.Distortion Intensity) {constB=0.1}
  NamedRerouteDeclaration(None=Saturate(None=Multiply(A=@n10, B=@n16))) {name=UV Distortion, nodeColor=<rgba 0.6406 0 1 1>}
 ]
}
