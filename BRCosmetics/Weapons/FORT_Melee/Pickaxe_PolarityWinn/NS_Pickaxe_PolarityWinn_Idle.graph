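# Niagara system for the idle glow of the PolarityWinn pickaxe, with User.Variant blending color palettes.
# BladeAura and AudioLines render two-layer meshes, while Glows, Hand_Decals and Light add sprites and a light.
# System.AudioReactivity boosts brightness and drives mesh material parameters; a fake reactivity loop module is referenced.
# Time-of-day multipliers scale color and alpha, and Glows varies sprite width by camera angle.
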NiagaraSystem /BRCosmetics/Weapons/FORT_Melee/Pickaxe_PolarityWinn/NS_Pickaxe_PolarityWinn_Idle
dependencies {
 usedDataInterfaces [
  &NiagaraDataInterfaceCurve
  &NiagaraDataInterfaceCamera
  &NiagaraDataInterfaceSocketReader
  &NiagaraDataInterfaceColorCurve
 ]
 usedDynamicInputs [
  &ColorFromCurve
  /Niagara/DynamicInputs/Lerp/SmoothLerpOverTime_Float
  &FN_FrontEndFloatModifier
  &RemapRange
  &MakeCustomFloatFromBool
  &SetBoolByFloatComparison
 ]
 usedModules [
  &self:SystemSpawnScript.SystemScriptSource.SystemScriptGraph.NiagaraNodeAssignment_0.SetVariables_#1
  &self:SystemSpawnScript.SystemScriptSource.SystemScriptGraph.NiagaraNodeAssignment_1.SetVariables_#2
  &self:SystemSpawnScript.SystemScriptSource.SystemScriptGraph.NiagaraNodeAssignment_4.SetVariables_#3
  &SystemState
  &self:SystemSpawnScript.SystemScriptSource.SystemScriptGraph.NiagaraNodeAssignment_5.SetVariables_#4
  /Game/Effects/Niagara/Modules/Audio/FN_FakeAudioReactivityLoop
  &self:SystemSpawnScript.SystemScriptSource.SystemScriptGraph.NiagaraNodeAssignment_3.SetVariables_#5
 ]
}
emitters {
 AudioLines {
  emitterSpawnScript [/SetVariables(
    Emitter.Lines Color Center:LinearColor=&LerpLinearColors(StartColor=<rgba 0.0518 0.1821 1 1>, EndColor=<rgba 0.0182 0.4156 1 1>, LerpFactor=$User.Variant)
    Emitter.Lines Color Inner:LinearColor=&LerpLinearColors(StartColor=<rgba 0.1598 0.8964 1 1>, EndColor=<rgba 0 1 1 1>, LerpFactor=$User.Variant)
    Emitter.Lines Color Outer:LinearColor=&LerpLinearColors(StartColor=<rgba 0.9155 0.0931 1 1>, EndColor=<rgba 0.5365 0.1942 1 1>, LerpFactor=$User.Variant)
   ) {isSetParameters=true, module=SetVariables_#6}]
  emitterUpdateScript [
   &EmitterState(
    Inactive Response=::NewEnumerator1
    Life Cycle Mode=::NewEnumerator1
    Loop Behavior=::NewEnumerator1
    Loop Duration Mode=::NewEnumerator1
    UseLoopDelay=false
   )
   &SpawnBurst_Instantaneous(Use Spawn Probability=false, UseLoopCountLimit=false)
   &SpawnBurst_Instantaneous(Spawn Group=1, Use Spawn Probability=false, UseLoopCountLimit=false) {module=SpawnBurst_Instantaneous001}
  ]
  particleSpawnScript [
   &InitializeParticle(
    Color=<rgba 2 2 2 1>
    Lifetime=$Emitter.CurrentLoopDuration
    Mesh Scale=&Multiply_VectorByFloat(Vector=<0.9 1.25 1>, Float=0.75)
    Mesh Scale Mode=::NewEnumerator4
    Position Mode=::NewEnumerator0
    Sprite Size Mode=::NewEnumerator3
    Write Material Random=true
   )
   &InitialMeshOrientation(
    Mesh Axis To Orient=<0 0 1>
    Mesh Orientation Mode=::NewEnumerator2
    Rotation=<-0.03 0 -0.215>
    Rotation Coordinate Space=::NewEnumerator2
    Use Rotation Vector=true
   ) {module=InitialMeshOrientation001}
   /SetVariables(Particles.VisibilityTag=&SpawnGroup()) {isSetParameters=true}
  ]
  particleUpdateScript [
   &ParticleState(Kill Particles When Lifetime Has Elapsed=false)
   &SolveForcesAndVelocity(Clamp Velocity=false, Limit Acceleration=false)
   /SetVariables(Particles.Position=&AddVectorToPosition(Position=$Engine.Emitter.SimulationPosition, Vector=<8 -3.5 -1.5>)) {isSetParameters=true, module=SetVariables_#7}
   &DynamicMaterialParameters(
    Index 0 Param 1=&FloatFromCurve(
     FloatCurve=DataInterface() {propertyValues {Curve {keys [<at 0 0.4 RCIM_Cubic>, {arriveTangent=1.0197, interpMode=RCIM_Cubic, leaveTangent=1.0197, tangentMode=RCTM_User, time=1, value=1.5}]}}}
     CurveIndex=$System.AudioReactivity
    )
    Index 0 Param 2=&Lerp_Float(
     A=&Multiply_Float(
      A=1.0147
      B=&Lerp_Float(
       A=1
       B=&Waveform("Waveform [1]"=::NewEnumerator2, "[1] Frequency"=1, "[1] Compound Frequency"=1, "[1] Amplitude Min/Max"=<-0.5 1.5>)
       Alpha=$System.AudioReactivity
      )
     )
     B=&Multiply_Float(
      A=0.4
      B=&Lerp_Float(
       A=1
       B=&Waveform("Waveform [1]"=::NewEnumerator2, "[1] Phase Scale"=2, "[1] Frequency"=1, "[1] Compound Frequency"=1, "[1] Amplitude Min/Max"=<-0.5 1.5>)
       Alpha=$System.AudioReactivity
      )
     )
     Alpha=&MakeFloatFromInt(Int=&SpawnGroup())
    )
    Index 0 Param 3=&FloatFromCurve(
     FloatCurve=DataInterface() {propertyValues {Curve {keys [<at 0 5 RCIM_Constant>, <at 0.5 6 RCIM_Constant>, {arriveTangent=1.0197, interpMode=RCIM_Constant, leaveTangent=1.0197, time=1, value=7}]}}}
     CurveIndex=$System.AudioReactivity
    )
    Module.Index0Float0Write=true
    Module.Index0Float1Write=true
    Module.Index0Float2Write=true
    Module.Index0Float3Write=true
   )
   &ScaleColor(
    Scale Alpha=&Multiply_Float(
     A=&Lerp_Float(
      A=1
      B=&Lerp_Float(A=1, B=0.05, Alpha=$System.AudioReactivity)
      Alpha=&MakeFloatFromInt(Int=&SpawnGroup())
     )
     B=&MakeFloatFromLinearColor(LinearColor=$System.TODColorMultiplier, Channel:ENiagaraFloatFromLinearColorOptions=::NewEnumerator3)
    )
    Scale RGB=&Multiply_Vector(
     A=&Lerp_Vector(
      A=&VectorFromFloat(Value=&MakeFloatFromVector(
        Vector=&VectorFromFloat(Value=&Lerp_Float(A=1, B=1.75, Alpha=$System.AudioReactivity))
        Channel=::NewEnumerator0
       ))
      B=&VectorFromFloat(Value=&MakeFloatFromVector(
        Vector=&VectorFromFloat(Value=&Lerp_Float(A=1, B=5, Alpha=$System.AudioReactivity))
        Channel=::NewEnumerator0
       ))
      Alpha=&MakeFloatFromInt(Int=&SpawnGroup())
     )
     B=&MakeVectorFromLinearColorRGB(Color=$System.TODColorMultiplier)
    )
    ScaleA=true
    ScaleRGB=true
   )
  ]
  rendererClasses [&NiagaraMeshRendererProperties]
  renderers [
   {rendererClass=&NiagaraMeshRendererProperties, rendererIndex=0}
   {rendererClass=&NiagaraMeshRendererProperties, rendererIndex=1}
  ]
 }
 BladeAura {
  emitterSpawnScript [
   /SetVariables(
    Emitter.Color Center:LinearColor=&LerpLinearColors(StartColor=<rgba 0.072 0.1286 0.9 1>, EndColor=<rgba 0 0.3755 0.429 1>, LerpFactor=$User.Variant)
    Emitter.Color Inner:LinearColor=&LerpLinearColors(StartColor=<rgba 0.1598 0.8964 1 1>, EndColor=<rgba 0.2761 0.8914 0.9 1>, LerpFactor=$User.Variant)
    Emitter.Color Outer:LinearColor=&LerpLinearColors(StartColor=<rgba 0.9155 0.0931 1 1>, EndColor=<rgba 1.2167 0.078 1.5 1>, LerpFactor=$User.Variant)
   ) {isSetParameters=true, module=SetVariables_#6}
   /SetVariables(
    Emitter.Color Center_OuterAura:LinearColor=&LerpLinearColors(StartColor=<rgba 0.0518 0.1821 1 1>, EndColor=<rgba 0.0518 0.1821 1 1>, LerpFactor=$User.Variant)
    Emitter.Color Inner_OuterAura:LinearColor=&LerpLinearColors(StartColor=<rgba 0.1598 0.8964 1 1>, EndColor=<rgba 0.1598 1 1 1>, LerpFactor=$User.Variant)
    Emitter.Color Outer_OuterAura:LinearColor=&LerpLinearColors(StartColor=<rgba 0.9514 0.4783 1 1>, EndColor=<rgba 1 0.1584 0.3545 1>, LerpFactor=$User.Variant)
   ) {isSetParameters=true}
  ]
  emitterUpdateScript [
   &EmitterState(
    Inactive Response=::NewEnumerator1
    Life Cycle Mode=::NewEnumerator1
    Loop Behavior=::NewEnumerator1
    Loop Duration Mode=::NewEnumerator1
    UseLoopDelay=false
   )
   &SpawnBurst_Instantaneous(Use Spawn Probability=false, UseLoopCountLimit=false)
   &SpawnBurst_Instantaneous(Spawn Group=1, Use Spawn Probability=false, UseLoopCountLimit=false) {module=SpawnBurst_Instantaneous001}
  ]
  particleSpawnScript [
   &InitializeParticle(
    Color=&LerpLinearColors(
     StartColor=<rgba 0.85 0.85 0.85 3>
     EndColor=<rgba 1 1 1 0.6>
     LerpFactor=&MakeFloatFromInt(Int=&SpawnGroup())
    )
    Lifetime=$Emitter.CurrentLoopDuration
    Mesh Scale=&Lerp_Vector(
     A=&Multiply_VectorByFloat(Vector=<1 1 1>, Float=1)
     B=&Multiply_VectorByFloat(Vector=<1 1 1>, Float=1.05)
     Alpha=&MakeFloatFromInt(Int=&SpawnGroup())
    )
    Mesh Scale Mode=::NewEnumerator4
    Position Mode=::NewEnumerator0
    Sprite Size Mode=::NewEnumerator3
    Write Material Random=true
   )
   &InitialMeshOrientation(
    Mesh Axis To Orient=<0 0 1>
    Mesh Orientation Mode=::NewEnumerator2
    Rotation=<-0.03 0 -0.215>
    Rotation Coordinate Space=::NewEnumerator2
    Use Rotation Vector=true
   ) {module=InitialMeshOrientation001}
   /SetVariables(Particles.VisibilityTag=&SpawnGroup()) {isSetParameters=true}
  ]
  particleUpdateScript [
   &ParticleState(Kill Particles When Lifetime Has Elapsed=false)
   &SolveForcesAndVelocity(Clamp Velocity=false, Limit Acceleration=false)
   /SetVariables(Particles.Position=&AddVectorToPosition(Position=$Engine.Emitter.SimulationPosition, Vector=<9.5 -2.3 -2>)) {isSetParameters=true, module=SetVariables_#7}
   &DynamicMaterialParameters(
    Index 0 Param 1=-0.5
    Index 0 Param 2=3
    Index 0 Param 3=&FloatFromCurve(FloatCurve=DataInterface() {propertyValues {Curve {keys [<at 0 0.7 RCIM_Cubic>,<at 1 0.5 RCIM_Cubic>]}}}, CurveIndex=$System.AudioReactivity)
    Index 0 Param 4=&FloatFromCurve(FloatCurve=DataInterface() {propertyValues {Curve {keys [<at 0 1.2 RCIM_Cubic>,<at 1 2 RCIM_Cubic>]}}}, CurveIndex=$System.AudioReactivity)
    Module.Index0Float0Write=true
    Module.Index0Float1Write=true
    Module.Index0Float2Write=true
    Module.Index0Float3Write=true
   )
   &ScaleColor(
    Scale Alpha=&MakeFloatFromLinearColor(LinearColor=$System.TODColorMultiplier, Channel:ENiagaraFloatFromLinearColorOptions=::NewEnumerator3)
    Scale RGB=&Multiply_Vector(
     A=&Lerp_Vector(
      A=&Lerp_Vector(
       A=&Multiply_Vector(
        A=&MakeVectorFromLinearColorRGB(Color=&LerpLinearColors(StartColor=<rgba 0.7 0.7251 1 1>, EndColor=<rgba 0.85 0.8625 1 1>, LerpFactor=$System.AudioReactivity))
        B=&VectorFromFloat(Value=&Lerp_Float(A=1.4, B=1.75, Alpha=$System.AudioReactivity))
       )
       B=&Multiply_Vector(
        A=&MakeVectorFromLinearColorRGB(Color=&LerpLinearColors(StartColor=<rgba 0.7 0.7251 1 1>, EndColor=<rgba 0.85 0.8625 1 1>, LerpFactor=$System.AudioReactivity))
        B=&VectorFromFloat(Value=&Lerp_Float(A=1.25, B=1.6, Alpha=$System.AudioReactivity))
       )
       Alpha=&MakeFloatFromInt(Int=&SpawnGroup())
      )
      B=&Lerp_Vector(
       A=&Multiply_Vector(
        A=&MakeVectorFromLinearColorRGB(Color=&LerpLinearColors(StartColor=<rgba 0.85 0.8625 1 1>, EndColor=<rgba 1 1 1 1>, LerpFactor=$System.AudioReactivity))
        B=&VectorFromFloat(Value=&Lerp_Float(A=1.4, B=1.75, Alpha=$System.AudioReactivity))
       )
       B=&Multiply_Vector(
        A=&MakeVectorFromLinearColorRGB(Color=&LerpLinearColors(StartColor=<rgba 0.85 1 0.985 1>, EndColor=<rgba 0.85 1 0.9861 1>, LerpFactor=$System.AudioReactivity))
        B=&VectorFromFloat(Value=&Lerp_Float(A=1.25, B=1.6, Alpha=$System.AudioReactivity))
       )
       Alpha=&MakeFloatFromInt(Int=&SpawnGroup())
      )
      Alpha=$User.Variant
     )
     B=&MakeVectorFromLinearColorRGB(Color=$System.TODColorMultiplier)
    )
    ScaleA=true
    ScaleRGB=true
   )
  ]
  rendererClasses [&NiagaraMeshRendererProperties]
  renderers [
   {rendererClass=&NiagaraMeshRendererProperties, rendererIndex=0}
   {rendererClass=&NiagaraMeshRendererProperties, rendererIndex=1}
  ]
 }
 Glows {
  emitterUpdateScript [
   &EmitterState(
    Inactive Response=::NewEnumerator1
    Life Cycle Mode=::NewEnumerator1
    Loop Behavior=::NewEnumerator1
    Loop Duration Mode=::NewEnumerator1
    UseLoopDelay=false
   )
   &SpawnBurst_Instantaneous(Spawn Count=2, Use Spawn Probability=false, UseLoopCountLimit=false)
  ]
  particleSpawnScript [
   &PartitionParticles(ID to Partition:NiagaraInt32=&ReturnExecIndex())
   &InitializeParticle(
    Color=&LerpLinearColors(
     StartColor=&MergeHelper_ColorSwitch(
      0=&MultiplyLinearColor(
       Linear Color=<rgba 0.1884 0.4704 1 0.2>
       Scale Factor=&MakeVector4_FromLinearColor(LinearColor=$System.TODColorMultiplier)
      )
      1=&MultiplyLinearColor(
       Linear Color=<rgba 0.4678 0.1946 0.6038 0.35>
       Scale Factor=&MakeVector4_FromLinearColor(LinearColor=$System.TODColorMultiplier)
      )
      2=?
      3=?
      Index=$Particles.PartitionParticles.Partition
      Num=::NewEnumerator0
     )
     EndColor=&MergeHelper_ColorSwitch(0=<rgba 0.1884 1 1 0.2>, 1=<rgba 0.5356 0.1946 0.6038 0.35>, 2=?, 3=?, Index=$Particles.PartitionParticles.Partition, Num=::NewEnumerator0)
     LerpFactor=$User.Variant
    )
    Lifetime=$Emitter.CurrentLoopDuration
    Sprite Rotation Mode=::NewEnumerator1
    Sprite Size=&MergeHelper_Vector2Switch(0=<20 40>, 1=<20 90>, 2=?, 3=?, 4=?, Index=$Particles.PartitionParticles.Partition, Num=::NewEnumerator0)
    Sprite Size Mode=::NewEnumerator4
    UsePositionOffset=false
    Write Material Random=true
   )
   &CameraOffset()
  ]
  particleUpdateScript [
   &ParticleState(Kill Particles When Lifetime Has Elapsed=false)
   /SetVariables(Particles.Position=&AddVectorToPosition(
     Position=$Engine.Emitter.SimulationPosition
     Vector=&MergeHelper_VectorSwitch(0=<1 -5 0>, 1=<-10.3192 -8 0>, 2=?, 3=?, Index=$Particles.PartitionParticles.Partition, Num=::NewEnumerator0)
    )) {isSetParameters=true, module=SetVariables_#7}
   &SpriteFacingAndAlignment(Sprite Alignment=<1 0.3 -0.13>, WriteAlignment=true, WriteFacing=false)
   &DynamicMaterialParameters(
    Index 0 Param 1=2
    Module.Index0Float0Write=true
    Module.Index0Float1Write=true
    Module.Index0Float2Write=true
    Module.Index0Float3Write=true
   )
   &ScaleColor(
    Scale Alpha=&MakeFloatFromLinearColor(LinearColor=$System.TODColorMultiplier, Channel:ENiagaraFloatFromLinearColorOptions=::NewEnumerator3)
    Scale RGB=&Multiply_Vector(
     A=&Multiply_VectorByFloat(
      Vector=&VectorFromFloat(Value=&Lerp_Float(A=1, B=1.75, Alpha=$System.AudioReactivity))
      Float=&FNTimeOfDayScalar(Night Scale=0.7, Morning Scale=0.7, Evening Scale=0.7)
     )
     B=&MakeVectorFromLinearColorRGB(Color=$System.TODColorMultiplier)
    )
    ScaleA=true
    ScaleRGB=true
   )
   /SetVariables(Particles.SpriteSize=&Multiply_Vector2D(
     A=$Particles.Initial.SpriteSize
     B=&MakeVector2D(
      X=&Lerp_Float(
       A=0.5
       B=1.8
       Alpha=&Abs_Float(Float:NiagaraFloat=&DotProduct(
         Vector A=&FN_FrontEndVector3Modifier(
          Gameplay=&TransformVector(Vector=<0 1 0>, Destination Space=::Simulation, Apply Scale=true)
          Front End=&TransformVector(Vector=<-0.25 1 0>, Destination Space=::Simulation, Apply Scale=true)
         )
         Vector B=&ReturnCameraVector(Camera Vectors=::NewEnumerator4, Coordinate Space=::Simulation, Origin Position=$Particles.Position)
        ))
      )
      Y=1
     )
    )) {isSetParameters=true}
  ]
  rendererClasses [&NiagaraSpriteRendererProperties]
  renderers [{rendererClass=&NiagaraSpriteRendererProperties}]
 }
 Hand_Decals {
  emitterUpdateScript [
   &EmitterState(
    Inactive Response=::NewEnumerator1
    Life Cycle Mode=::NewEnumerator1
    Loop Behavior=::NewEnumerator1
    Loop Duration Mode=::NewEnumerator1
    UseLoopDelay=false
   )
   &SpawnBurst_Instantaneous(Spawn Count=2, Use Spawn Probability=false, UseLoopCountLimit=false)
  ]
  particleSpawnScript [
   &PartitionParticles()
   &InitializeParticle(
    Color=&LerpLinearColors(
     StartColor=&MergeHelper_ColorSwitch(
      0=&MultiplyLinearColor(
       Linear Color=<rgba 0.8521 7.2583 10 1>
       Scale Factor=&MakeVector4_FromLinearColor(LinearColor=$System.TODColorMultiplier)
      )
      1=&MultiplyLinearColor(
       Linear Color=<rgba 1.488 0.7336 2 1>
       Scale Factor=&MakeVector4_FromLinearColor(LinearColor=$System.TODColorMultiplier)
      )
      2=?
      3=?
      Index=$Particles.PartitionParticles.Partition
      Num=::NewEnumerator0
     )
     EndColor=&MergeHelper_ColorSwitch(
      0=&MultiplyLinearColor(
       Linear Color=<rgba 0.8521 10 10 1>
       Scale Factor=&MakeVector4_FromLinearColor(LinearColor=$System.TODColorMultiplier)
      )
      1=&MultiplyLinearColor(
       Linear Color=<rgba 0.5 0.068 0.5 1>
       Scale Factor=&MakeVector4_FromLinearColor(LinearColor=$System.TODColorMultiplier)
      )
      2=?
      3=?
      Index=$Particles.PartitionParticles.Partition
      Num=::NewEnumerator0
     )
     LerpFactor=$User.Variant
    )
    Sprite Size Mode=::NewEnumerator3
    Uniform Sprite Size=&MergeHelper_FloatSwitch(0=25, 1=20, 2=?, 3=?, Index=$Particles.PartitionParticles.Partition, Num=::NewEnumerator0)
    UsePositionOffset=false
    Write Material Random=true
   )
   &ShapeLocation(MaskSpawnGroup=false, Override Seed=false, Sphere Radius=1, UseOwnerScale=true) {module=ShapeLocation001}
   &CameraOffset(Camera Offset Amount=30)
  ]
  particleUpdateScript [
   &ParticleState(Kill Particles When Lifetime Has Elapsed=false)
   &ScaleColor(
    Scale Alpha=&MakeFloatFromLinearColor(LinearColor=$System.TODColorMultiplier, Channel:ENiagaraFloatFromLinearColorOptions=::NewEnumerator3)
    Scale RGB=&Multiply_Vector(
     A=&Multiply_VectorByFloat(
      Vector=&VectorFromFloat(Value=&Lerp_Float(A=1, B=1.75, Alpha=$System.AudioReactivity))
      Float=&FNTimeOfDayScalar(Night Scale=0.2, Morning Scale=0.2, Evening Scale=0.4)
     )
     B=&MakeVectorFromLinearColorRGB(Color=$System.TODColorMultiplier)
    )
    ScaleA=true
    ScaleRGB=true
   )
   /SetVariables(Particles.Position=&AddVectorToPosition(
     Position=$Engine.Emitter.SimulationPosition
     Vector=&MergeHelper_VectorSwitch(0=<2 -7 0>, 1=<12 5 0>, 2=?, 3=?, Index=$Particles.PartitionParticles.Partition, Num=::NewEnumerator0)
    )) {isSetParameters=true, module=SetVariables_#7}
  ]
  rendererClasses [&NiagaraSpriteRendererProperties]
  renderers [
   {rendererClass=&NiagaraSpriteRendererProperties}
   {rendererClass=&NiagaraSpriteRendererProperties, rendererIndex=1}
  ]
 }
 Light {
  emitterUpdateScript [
   &EmitterState(
    Inactive Response=::NewEnumerator1
    Life Cycle Mode=::NewEnumerator1
    Loop Behavior=::NewEnumerator1
    Loop Duration Mode=::NewEnumerator1
    UseLoopDelay=false
   )
   &SpawnBurst_Instantaneous(Use Spawn Probability=false, UseLoopCountLimit=false)
   &ShapeLocation(MaskSpawnGroup=false, Override Seed=false, Sphere Radius=1, UseOwnerScale=true) {module=ShapeLocation001}
  ]
  particleSpawnScript [
   &InitializeParticle(
    Color=&LerpLinearColors(StartColor=<rgba 0.8521 7.2583 10 1>, EndColor=<rgba 0.426 5 5 1>, LerpFactor=$User.Variant)
    Sprite Size Mode=::NewEnumerator3
    Uniform Sprite Size=25
    UsePositionOffset=false
    Write Material Random=true
   )
   /SetVariables(Particles.LightExponent=4) {isSetParameters=true}
   /SetVariables(Particles.LightRadius=60) {isSetParameters=true}
  ]
  particleUpdateScript [
   &ParticleState(Kill Particles When Lifetime Has Elapsed=false)
   &ScaleColor(
    Scale Alpha=&MakeFloatFromLinearColor(LinearColor=$System.TODColorMultiplier, Channel:ENiagaraFloatFromLinearColorOptions=::NewEnumerator3)
    Scale RGB=&Multiply_Vector(
     A=&Multiply_VectorByFloat(
      Vector=&VectorFromFloat(Value=&Lerp_Float(A=1, B=3, Alpha=$System.AudioReactivity))
      Float=&FNTimeOfDayScalar(Night Scale=0.7, Morning Scale=0.3, Evening Scale=0.5)
     )
     B=&MakeVectorFromLinearColorRGB(Color=$System.TODColorMultiplier)
    )
    ScaleA=true
    ScaleRGB=true
   )
   /SetVariables(Particles.Position=&AddVectorToPosition(Position=$Engine.Emitter.SimulationPosition, Vector=<4 -5 0>)) {isSetParameters=true, module=SetVariables_#7}
  ]
  rendererClasses [&NiagaraLightRendererProperties, &NiagaraSpriteRendererProperties]
  renderers [
   {rendererClass=&NiagaraLightRendererProperties, rendererIndex=0}
   {rendererClass=&NiagaraSpriteRendererProperties, rendererIndex=1}
  ]
 }
}
summary {
 systemName=NS_Pickaxe_PolarityWinn_Idle
 emitters [
  {emitterName=BladeAura, rendererClasses [&NiagaraMeshRendererProperties]}
  {emitterName=AudioLines, rendererClasses [&NiagaraMeshRendererProperties]}
  {emitterName=Light, rendererClasses [&NiagaraLightRendererProperties, &NiagaraSpriteRendererProperties]}
  {emitterName=Glows, rendererClasses [&NiagaraSpriteRendererProperties]}
  {emitterName=Hand_Decals, rendererClasses [&NiagaraSpriteRendererProperties]}
 ]
 userVariables [
  User.bIsFrontEnd {type=NiagaraBool}
  User.bIsFrontEndPreview {type=NiagaraBool}
  User.DebugReactivity {type=NiagaraBool}
  User.EnableFrontendFakeAudioLoop {type=NiagaraFloat, defaultValue=0, description=""}
  User.ReactivityEnabled {type=NiagaraFloat, defaultValue=0, description=""}
  User.Variant {type=NiagaraFloat, defaultValue=0, description=""}
 ]
}
